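# EQS query that finds candidate Sprite spawn locations for the Fire Petal Day Night Forest content.
# A coarse, randomly jittered grid around the querier, sized by a query parameter and spaced by a scalable float, projects onto navigation using the default Sprites nav filter.
# Tests then filter points on building interior and water, and check that the surrounding surface is flat.
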EnvQuery /FirePetalDayNightForest/Gameplay/EQS/EQS_DayNightForest_SpriteSpawn
options [{
  generator {
   name=FortEnvQueryGenerator_SimpleGrid_2
   class=FortEnvQueryGenerator_SimpleGrid
   properties {
    bAutoSortTests=true
    bCanRunAsync=false
    envQueryResultNormalizationOption=Default
    generateAround=&EnvQueryContext_Querier
    navDataOverrideContext=null
    optionName=""
    overridenAgentNameForNavmesh=None
    randomDisplacementRatio=0.35
    gridSize {
     dataBinding=&self:FortEnvQueryGenerator_SimpleGrid_2.AIDataProvider_QueryParams_0
     dataField=FloatValue
     defaultValue=100000
    }
    projectionData {
     bCanDisableTrace=true
     bCanProjectDown=true
     bCanTraceOnGeometry=true
     bCanTraceOnNavMesh=true
     bOnlyBlockingHits=true
     bTraceComplex=false
     extentX=0
     extentY=10
     extentZ=10
     navigationFilter=/Sprites/AI/NavFilters/NavFilter_DefaultSprites.NavFilter_DefaultSprites_C
     postProjectionVerticalOffset=0
     projectDown=10000
     projectUp=10000
     serializedChannel=ECC_GameTraceChannel11
     traceChannel=TraceTypeQuery10
     traceMode=Navigation
     traceProfileName=None
     traceShape=Line
     versionNum=1
    }
    spaceBetween {
     dataBinding=&self:FortEnvQueryGenerator_SimpleGrid_2.FortAIDataProvider_ScalableFloat_0
     dataField=FloatValue
     defaultValue=25000
    }
   }
  }
  tests [
   FortQueryTest_InsideBuilding_3 {class=FortQueryTest_InsideBuilding, properties {testPurpose=Filter}}
   FortQueryTest_InsideWater_1 {
    class=FortQueryTest_InsideWater
    properties {
     bDefineReferenceValue=false
     boolValue=false
     clampMaxType=None
     clampMinType=None
     filterType=Match
     floatValueMax=0
     floatValueMin=0
     multipleContextFilterOp=AllPass
     multipleContextScoreOp=AverageScore
     normalizationType=Absolute
     referenceValue=0
     scoreClampMax=0
     scoreClampMin=0
     scoringEquation=Constant
     scoringFactor=1
     testComment=""
     testPurpose=Filter
     testRadius=128
    }
   }
   FortQueryTest_OnFlatSurface_1 {class=FortQueryTest_OnFlatSurface, properties {radius=200, toleranceZ=120, traceOffsetUp=100}}
  ]
 }]
